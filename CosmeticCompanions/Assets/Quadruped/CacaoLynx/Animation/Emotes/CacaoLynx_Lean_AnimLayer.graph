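# Animation layer Blueprint for the CacaoLynx cosmetic companion emote on the player skeleton, adding turn lean and ground-slope tilt.
# Each update, Lean smooths yaw rate into LeanAngle, while UpdateSlopeCorrection traces below player 0, not the owning actor, smoothing the local ground normal into pitch and roll.
# ItemFinalPoseOverride applies a LeanAngle-driven additive blendspace to the incoming pose, then additively rotates Spine_01 by the slope correction.
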AnimBlueprint /CosmeticCompanions/Assets/Quadruped/CacaoLynx/Animation/Emotes/CacaoLynx_Lean_AnimLayer
target_skeleton=&Fortnite_M_Avg_Player_Skeleton
graphs [
 AnimGraph {nodes [
   Blend Poses by bool(BlendPose_0=Slot 'InMotion'() {slotName=InMotion}, BlendPose_1=Slot 'DefaultSlot'(), BlendTime_0=0.3, BlendTime_1=0.3)
   Look At(Alpha=1) {boneToModify=~, lookAtTarget=~}
   AnimGraphNode_BlendSpaceGraph_0 = Blendspace 'BS_LeanTest'!(X:float=Get LeanAngle() -> (LeanAngle:double)) -> (Pose:Pose Link Structure) {is=title AnimGraphNode_BlendSpaceGraph}
   Output Pose(Result=Apply Additive(Additive=@AnimGraphNode_BlendSpaceGraph_0))
   Two Bone IK!(ComponentPose:Component Space Pose Link Structure, Alpha:float=1, EffectorLocation:Vector="0,0,0", JointTargetLocation:Vector="0,0,0") -> (Pose:Component Space Pose Link Structure) {is=title AnimGraphNode_TwoBoneIK}
   PrintString!(execute, InString=Hello, bPrintToScreen, bPrintToLog, TextColor, Duration, Key) -> (then)
  ]}
 BS_LeanTest {path=AnimGraph.AnimGraphNode_BlendSpaceGraph_0.BS_LeanTest, links [], nodes []}
 Emote_Suspenders_CMF_LeanLeft {
  path=AnimGraph.AnimGraphNode_BlendSpaceGraph_0.BS_LeanTest.Emote_Suspenders_CMF_LeanLeft
  nodes [Output Animation Pose!(Result:Pose Link Structure="Sequence Player (additive) 'Emote_Suspenders_CMF_LeanLeft'"() {method=Graph, sequence=/Game/Animation/Game/MainPlayer/Emotes/Suspenders/Emote_Suspenders_CMF_LeanLeft}) {is=title AnimGraphNode_BlendSpaceSampleResult}]
 }
 Emote_Suspenders_CMF_LeanRight {
  path=AnimGraph.AnimGraphNode_BlendSpaceGraph_0.BS_LeanTest.Emote_Suspenders_CMF_LeanRight
  nodes [Output Animation Pose(Result="Sequence Player (additive) 'Emote_Suspenders_CMF_LeanRight'"() {method=Graph, sequence=/Game/Animation/Game/MainPlayer/Emotes/Suspenders/Emote_Suspenders_CMF_LeanRight})]
 }
 Emote_Suspenders_CMF_NoLean {
  path=AnimGraph.AnimGraphNode_BlendSpaceGraph_0.BS_LeanTest.Emote_Suspenders_CMF_NoLean
  nodes [Output Animation Pose(Result="Sequence Player (additive) 'Emote_Suspenders_CMF_NoLean'"() {method=Graph, sequence=/Game/Animation/Game/MainPlayer/Emotes/Suspenders/Emote_Suspenders_CMF_NoLean})]
 }
 EventGraph {nodes [
   n1 = Event BlueprintUpdateAnimation()
   > Lean!(execute, self, OwningActor:Actor Object Reference=Get Actor() -> (Actor:Actor Object Reference), DeltaTime:double=@n1.DeltaTimeX) -> (then)
   > UpdateSlopeCorrection!(
    execute
    self
    OwningActor:Actor Object Reference=Get Actor()
    DeltaTime:double=@n1.DeltaTimeX
    MeshComp:Skeletal Mesh Component Object Reference
   ) -> (then)
   Event BlueprintInitializeAnimation()
   > Set Actor(Actor:Actor Object Reference=GetOwningActor()) -> (Output_Get:Actor Object Reference)
   Event BlueprintBeginPlay() {is=title K2Node_Event}
  ]}
 ItemAimOffset {nodes [
   Input Pose!() -> (Pose, AimOffsetAlpha:double, Yaw:double, Pitch:double) {name=InPose_AimOffset}
   Input Pose!() -> (Pose) {name=InPose_PreAimOffsets}
   Output Pose() {layerGroup=Item Override}
  ]}
 ItemFallAdditive {nodes [
   Input Pose() {name=InPoseItemFallAdditive}
   Output Pose() {layerGroup=Item Override}
  ]}
 ItemFallLandAdditive {nodes [
   Input Pose() {name=InPoseItemFallLandAdditive}
   Output Pose() {layerGroup=Item Override}
  ]}
 ItemFallLevel2Override {nodes [
   Input Pose() {name=InPose_FallLevel2}
   Output Pose() {layerGroup=Item Override}
  ]}
 ItemFinalPoseOverride {nodes [
   AnimGraphNode_BlendSpaceGraph_1 = Blendspace 'CacaoLynx_AdditiveLean_BS'!(X:float=0) -> (Pose:Pose Link Structure) {is=title AnimGraphNode_BlendSpaceGraph}
   n1 = Get LeanAngle()
   "Transform (Modify) Bone - Bone: head"!(ComponentPose, Alpha=10, Translation="0,0,0", Rotation_Roll:float=0, Rotation_Pitch:float=0, Rotation_Yaw:float=@n1, Scale="1,1,1") -> (Pose) {rotationMode=BMM_Additive, translationSpace=BCS_BoneSpace, boneToModify {boneName=head}}
   AnimGraphNode_BlendSpaceGraph_3 = Blendspace 'CacaoLynx_AdditiveLean_BS'(X=@n1)
   n2 = Local To Component(LocalPose=@AnimGraphNode_BlendSpaceGraph_3)
   n3 = Local To Component(LocalPose=Apply Additive(Base=Input Pose() {name=InFinalPose}, Additive=Component To Local(ComponentPose=@n2)))
   n4 = Get StoreRollCorrection() -> (StoreRollCorrection:double)
   n5 = Get StorePitchCorrection() -> (StorePitchCorrection:double)
   Output Pose(Result=Component To Local(ComponentPose="Transform (Modify) Bone - Bone: Spine_01"!(ComponentPose=@n3, Alpha=0.75, Translation="0,0,0", Rotation_Roll:float=@n4, Rotation_Pitch:float=@n5, Rotation_Yaw:float=0, Scale="1,1,1") -> (Pose) {rotationMode=BMM_Additive, translationSpace=BCS_BoneSpace, boneToModify {boneName=Spine_01}})) {layerGroup=Item Override}
  ]}
 CacaoLynx_AdditiveLean_BS {path=ItemFinalPoseOverride.AnimGraphNode_BlendSpaceGraph_1.CacaoLynx_AdditiveLean_BS, links [], nodes []}
 CacaoLynx_Emote_Player_LeanLeft {
  path=ItemFinalPoseOverride.AnimGraphNode_BlendSpaceGraph_1.CacaoLynx_AdditiveLean_BS.CacaoLynx_Emote_Player_LeanLeft
  nodes [Output Animation Pose(Result="Sequence Player (additive) 'CacaoLynx_Emote_Player_LeanLeft'"() {method=Graph, sequence=&/CacaoLynx_Emote_Player_LeanLeft})]
 }
 CacaoLynx_Emote_Player_LeanRight {
  path=ItemFinalPoseOverride.AnimGraphNode_BlendSpaceGraph_1.CacaoLynx_AdditiveLean_BS.CacaoLynx_Emote_Player_LeanRight
  nodes [Output Animation Pose(Result="Sequence Player (additive) 'CacaoLynx_Emote_Player_LeanRight'"() {method=Graph, sequence=&/CacaoLynx_Emote_Player_LeanRight})]
 }
 CacaoLynx_Emote_Player_NoLean {
  path=ItemFinalPoseOverride.AnimGraphNode_BlendSpaceGraph_1.CacaoLynx_AdditiveLean_BS.CacaoLynx_Emote_Player_NoLean
  nodes [Output Animation Pose(Result="Sequence Player (additive) 'CacaoLynx_Emote_Player_NoLean'"() {method=Graph, sequence=&/CacaoLynx_Emote_Player_NoLean})]
 }
 CacaoLynx_AdditiveLean_BS {path=ItemFinalPoseOverride.AnimGraphNode_BlendSpaceGraph_3.CacaoLynx_AdditiveLean_BS, links [], nodes []}
 CacaoLynx_Emote_Player_LeanLeft {
  path=ItemFinalPoseOverride.AnimGraphNode_BlendSpaceGraph_3.CacaoLynx_AdditiveLean_BS.CacaoLynx_Emote_Player_LeanLeft
  nodes [Output Animation Pose(Result="Sequence Player (additive) 'CacaoLynx_Emote_Player_LeanLeft'"() {method=Graph, sequence=&CacaoLynx_Emote_Player_LeanLeft})]
 }
 CacaoLynx_Emote_Player_LeanRight {
  path=ItemFinalPoseOverride.AnimGraphNode_BlendSpaceGraph_3.CacaoLynx_AdditiveLean_BS.CacaoLynx_Emote_Player_LeanRight
  nodes [Output Animation Pose(Result="Sequence Player (additive) 'CacaoLynx_Emote_Player_LeanRight'"() {method=Graph, sequence=&CacaoLynx_Emote_Player_LeanRight})]
 }
 CacaoLynx_Emote_Player_NoLean {
  path=ItemFinalPoseOverride.AnimGraphNode_BlendSpaceGraph_3.CacaoLynx_AdditiveLean_BS.CacaoLynx_Emote_Player_NoLean
  nodes [Output Animation Pose(Result="Sequence Player (additive) 'CacaoLynx_Emote_Player_NoLean'"() {method=Graph, sequence=&CacaoLynx_Emote_Player_NoLean})]
 }
 ItemFullBodyAimOffsets {nodes [
   Input Pose!() -> (Pose, FullBody_AimOffsetAlpha:double, Aim_Yaw:double, Aim_Pitch:double) {name=InPoseFullBodyAimOffsets}
   Input Pose!() -> (Pose) {name=InPosePreFullBodyAimOffsets}
   Output Pose() {layerGroup=Item Override}
  ]}
 ItemFullBodyOverride {nodes [
   Input Pose() {name=InPoseFullBody}
   Output Pose() {layerGroup=Item Override}
  ]}
 ItemIdleOverride {nodes [
   Input Pose() {name=InPose_Idle}
   Output Pose() {layerGroup=Item Override}
  ]}
 ItemJetPackJumpAdditive {nodes [Output Pose() {layerGroup=Item Override}]}
 ItemJetPackStartAdditive {nodes [Output Pose() {layerGroup=Item Override}]}
 ItemJumpLoopAdditive {nodes [
   Input Pose() {name=InPoseItemJumpLoopAdditive}
   Output Pose() {layerGroup=Item Override}
  ]}
 ItemJumpOverride {nodes [
   Input Pose() {name=InPoseJump}
   Output Pose() {layerGroup=Item Override}
  ]}
 ItemJumpUpAdditive {nodes [
   Input Pose() {name=InPoseItemJumpUpAdditive}
   Output Pose() {layerGroup=Item Override}
  ]}
 ItemPreAimOffsets {nodes [
   Input Pose() {name=InPosePreAimOffsets}
   Output Pose() {layerGroup=Item Override}
  ]}
 ItemPreIK {nodes [
   Input Pose() {name=InPose_PreIK}
   Output Pose() {layerGroup=Item Override}
  ]}
 ItemRightHandIK {nodes [
   Input Pose() {name=InPosePreRightHandIK}
   Input Pose() {name=InPosePostRightHandIK}
   Output Pose() {layerGroup=Item Override}
  ]}
 ItemSkeletalControl {nodes [
   Input Pose()
   Output Pose() {layerGroup=Item Override}
  ]}
 ItemSwimJumpFallLoop {nodes [Output Pose() {layerGroup=Item Override}]}
 ItemSwimJumpSurfaceEnd {nodes [Output Pose() {layerGroup=Item Override}]}
 ItemSwimJumpSurfaceLoop {nodes [Output Pose() {layerGroup=Item Override}]}
 ItemUpperBody {nodes [
   Input Pose!() -> (Pose, InputParam:Fort Anim Input Adjusted Aim Structure) {name=InPoseUpperBody}
   Output Pose() {layerGroup=Item Override}
  ]}
 ItemZipLineStartAdditive {nodes [Output Pose() {layerGroup=Item Override}]}
 Lean {nodes [
   n1 = Lean!() -> (then, OwningActor:Actor Object Reference, DeltaTime:double=~) {is=title K2Node_FunctionEntry}
   > Set PreviousCharacterYaw(PreviousCharacterYaw:double=Get CharacterYaw() -> (CharacterYaw:double)) -> (Output_Get:double)
   > Set CharacterYaw(CharacterYaw:double=BreakRotator(InRot=Get Actor Rotation!(self:Actor Object Reference=@n1) -> (ReturnValue:Rotator) {is=title K2Node_CallFunction}).Yaw) -> (Output_Get:double)
   n2 = Get CharacterYaw()
   n3 = Get PreviousCharacterYaw() -> (PreviousCharacterYaw:double)
   > n4 = Set CharacterYawDelta(CharacterYawDelta:double=NormalizeAxis!(Angle:float=float - float(A=@n2, B=@n3)) -> (ReturnValue:float)) -> (Output_Get:double)
   n5 = SafeDivide(A=@n4, B=@n1.DeltaTime)
   > Set LeanAngle(LeanAngle:double=FInterpTo(Current=Get LeanAngle(), Target=float / float(A=@n5, B=425), DeltaTime=@n1.DeltaTime, InterpSpeed=5)) -> (Output_Get:double)
   > Return Node()
  ]}
 UpdateSlopeCorrection {nodes [
   n1 = UpdateSlopeCorrection!() -> (then, OwningActor:Actor Object Reference, DeltaTime:double=~, MeshComp:Skeletal Mesh Component Object Reference) {is=title K2Node_FunctionEntry}
   n2 = Get Actor Location(self=GetPlayerCharacter!(PlayerIndex:int=0) -> (ReturnValue:Character Object Reference))
   > n3 = Line Trace By Channel!(
    execute
    Start:Vector=@n2
    End:Vector=vector - vector(A=@n2, B=MakeVector(Y=0, Z=500))
    TraceChannel:ETraceTypeQuery Enum
    bTraceComplex:bool=false
    ActorsToIgnore:Array of Actor Object References
    DrawDebugType:EDrawDebugTrace Enum
    bIgnoreSelf:bool
    TraceColor:Linear Color Structure
    TraceHitColor:Linear Color Structure
    DrawTime:float
   ) -> (then, OutHit:Hit Result Structure, ReturnValue:bool)
   n4 = GetPlayerCharacter(PlayerIndex=0)
   n5 = BreakVector(InVec=Unrotate Vector!(
     A:Vector=BreakHitResult!(Hit:"Hit Result Structure (by ref)"=@n3) -> (
      bBlockingHit:bool
      bInitialOverlap:bool
      Time:float
      Distance:float
      Location:Vector
      ImpactPoint:Vector
      Normal:Vector
      ImpactNormal:Vector
      PhysMat:Physical Material Object Reference
      HitActor:Actor Object Reference
      HitComponent:Primitive Component Object Reference
      HitBoneName:Name
      BoneName:Name
      HitItem:int
      ElementIndex:int
      FaceIndex:int
      TraceStart:Vector
      TraceEnd:Vector
     ).ImpactNormal
     B:Rotator=Get Actor Rotation(self=@n4)
    ) -> (ReturnValue:Vector))
   > Set StorePitchCorrection(StorePitchCorrection:double=FInterpTo(
     Current=Get StorePitchCorrection()
     Target=float * float(A=@n5.Y, B=-90) {is=title K2Node_PromotableOperator}
     DeltaTime=@n1.DeltaTime
     InterpSpeed=5
    )) -> (Output_Get:double)
   > Set StoreRollCorrection(StoreRollCorrection:double=FInterpTo(Current=Get StoreRollCorrection(), Target=float * float(A=@n5, B=-90), DeltaTime=@n1.DeltaTime, InterpSpeed=5)) -> (Output_Get:double)
  ]}
]
variables [
 Actor {type=Actor, default=null}
 ActorRotation {type=number}
 AntiSlopeAlpha {type=number, default=0.25}
 bIsSkirted {type=boolean}
 CharacterYaw {type=number}
 CharacterYawDelta {type=number}
 DeltaTime {type=number}
 LeanAngle {type=number}
 PreviousCharacterYaw {type=number}
 StorePitchCorrection {type=number}
 StoreRollCorrection {type=number}
 Target {type=SkeletalMeshComponent}
 WheelRotation {type=number}
]
